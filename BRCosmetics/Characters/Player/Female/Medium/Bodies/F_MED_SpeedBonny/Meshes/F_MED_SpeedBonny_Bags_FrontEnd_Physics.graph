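# Physics asset for the front-end Bags mesh of the medium female SpeedBonny body, defining capsule bodies for dynamic jacket bones.
# Seven jacket chains (front, side and back) each link a tiny MSTR root capsule to three capsules through angular-limited constraints, with limits differing between left and right.
# Pelvis, spine_01 and both thighs add body capsules joined by free angular constraints, with no constraints to the jacket chains.
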PhysicsAsset /BRCosmetics/Characters/Player/Female/Medium/Bodies/F_MED_SpeedBonny/Meshes/F_MED_SpeedBonny_Bags_FrontEnd_Physics
bodies {
 dyn_jacket_01_L [{shapeType=Capsule, length=9.5, radius=4, shapeName=dyn_jacket_bk_01_R_capsule, center=<6.8195 0 0>, rotation=<rot -0.0004 -0.0003 -90>}]
 dyn_jacket_01_R [{shapeType=Capsule, length=9.5, radius=4, shapeName=dyn_jacket_01_R_capsule, center=<6.8195 0 0>, rotation=<rot 0 -0.0005 -89.9999>}]
 dyn_jacket_02_L [{shapeType=Capsule, length=9.5, radius=4, shapeName=dyn_jacket_bk_02_R_capsule, center=<6.8195 0 0>, rotation=<rot -0.0004 -0.0003 -90>}]
 dyn_jacket_02_R [{shapeType=Capsule, length=9.5, radius=4, shapeName=dyn_jacket_02_R_capsule, center=<6.8195 0 0>, rotation=<rot 0 -0.0005 -89.9999>}]
 dyn_jacket_03_L [{shapeType=Capsule, length=9.5, radius=4, shapeName=dyn_jacket_bk_03_R_capsule, center=<6.8195 0 0>, rotation=<rot -0.0004 -0.0003 -90>}]
 dyn_jacket_03_R [{shapeType=Capsule, length=9.5, radius=4, shapeName=dyn_jacket_03_R_capsule, center=<6.8195 0 0>, rotation=<rot 0 -0.0005 -89.9999>}]
 dyn_jacket_MSTR_L [{shapeType=Capsule, length=0, radius=0.5, shapeName=dyn_jacket_MSTR_L_capsule, rotation=<rot 0 0.0005 -90>}]
 dyn_jacket_MSTR_R [{shapeType=Capsule, length=0, radius=0.5, shapeName=dyn_jacket_MSTR_R_capsule, rotation=<rot 0 -0.0005 -89.9999>}]
 dyn_jacket_bk_01 [{shapeType=Capsule, length=9.5, radius=4, shapeName=dyn_jacket_bk_01_capsule, center=<6.8195 0 0>, rotation=<rot -0.0005 0 -90>}]
 dyn_jacket_bk_01_L [{shapeType=Capsule, length=9.5, radius=4, shapeName=dyn_jacket_bk_01_R_capsule, center=<6.8195 0 0>, rotation=<rot -0.0004 -0.0003 -90>}]
 dyn_jacket_bk_01_R [{shapeType=Capsule, length=9.5, radius=4, shapeName=dyn_jacket_bk_01_R_capsule, center=<6.8195 0 0>, rotation=<rot -0.0004 -0.0003 -90>}]
 dyn_jacket_bk_02 [{shapeType=Capsule, length=9.5, radius=4, shapeName=dyn_jacket_bk_02_capsule, center=<6.8195 0 0>, rotation=<rot -0.0005 0 -90>}]
 dyn_jacket_bk_02_L [{shapeType=Capsule, length=9.5, radius=4, shapeName=dyn_jacket_bk_02_R_capsule, center=<6.8195 0 0>, rotation=<rot -0.0004 -0.0003 -90>}]
 dyn_jacket_bk_02_R [{shapeType=Capsule, length=9.5, radius=4, shapeName=dyn_jacket_bk_02_R_capsule, center=<6.8195 0 0>, rotation=<rot -0.0004 -0.0003 -90>}]
 dyn_jacket_bk_03 [{shapeType=Capsule, length=9.5, radius=4, shapeName=dyn_jacket_bk_03_capsule, center=<6.8195 0 0>, rotation=<rot -0.0005 0 -90>}]
 dyn_jacket_bk_03_L [{shapeType=Capsule, length=9.5, radius=4, shapeName=dyn_jacket_bk_03_R_capsule, center=<6.8195 0 0>, rotation=<rot -0.0004 -0.0003 -90>}]
 dyn_jacket_bk_03_R [{shapeType=Capsule, length=9.5, radius=4, shapeName=dyn_jacket_bk_03_R_capsule, center=<6.8195 0 0>, rotation=<rot -0.0004 -0.0003 -90>}]
 dyn_jacket_bk_MSTR [{shapeType=Capsule, length=0, radius=0.5, shapeName=dyn_jacket_bk_MSTR_capsule, rotation=<rot -0.0005 0 -90>}]
 dyn_jacket_bk_MSTR_L [{shapeType=Capsule, length=0, radius=0.5, shapeName=dyn_jacket_bk_MSTR_L_capsule, rotation=<rot -0.0004 0.0003 -90>}]
 dyn_jacket_bk_MSTR_R [{shapeType=Capsule, length=0, radius=0.5, shapeName=dyn_jacket_bk_MSTR_R_capsule, rotation=<rot -0.0004 -0.0003 -90>}]
 dyn_jacket_fr_01_L [{shapeType=Capsule, length=9.5, radius=4, shapeName=dyn_jacket_bk_01_R_capsule, center=<6.8195 0 0>, rotation=<rot -0.0004 -0.0003 -90>}]
 dyn_jacket_fr_01_R [{shapeType=Capsule, length=9.5, radius=4, shapeName=dyn_jacket_fr_01_R_capsule, center=<6.8195 0 0>, rotation=<rot 0.0003 -0.0003 -89.9999>}]
 dyn_jacket_fr_02_L [{shapeType=Capsule, length=9.5, radius=4, shapeName=dyn_jacket_bk_02_R_capsule, center=<6.8195 0 0>, rotation=<rot -0.0004 -0.0003 -90>}]
 dyn_jacket_fr_02_R [{shapeType=Capsule, length=9.5, radius=4, shapeName=dyn_jacket_fr_02_R_capsule, center=<6.8195 0 0>, rotation=<rot 0.0003 -0.0003 -89.9999>}]
 dyn_jacket_fr_03_L [{shapeType=Capsule, length=9.5, radius=4, shapeName=dyn_jacket_bk_03_R_capsule, center=<6.8195 0 0>, rotation=<rot -0.0004 -0.0003 -90>}]
 dyn_jacket_fr_03_R [{shapeType=Capsule, length=9.5, radius=4, shapeName=dyn_jacket_fr_03_R_capsule, center=<6.8195 0 0>, rotation=<rot 0.0003 -0.0003 -89.9999>}]
 dyn_jacket_fr_MSTR_L [{shapeType=Capsule, length=0, radius=0.5, shapeName=dyn_jacket_fr_MSTR_L_capsule, rotation=<rot 0.0001 0.0001 0>}]
 dyn_jacket_fr_MSTR_R [{shapeType=Capsule, length=0, radius=0.5, shapeName=dyn_jacket_fr_MSTR_R_capsule, rotation=<rot 0.0003 -0.0003 -89.9999>}]
 pelvis [{shapeType=Capsule, length=8, radius=9.5, shapeName=pelvis_capsule, center=<-7.1805 -1.0007 0>}]
 spine_01 [{shapeType=Capsule, length=8.5, radius=8, shapeName=spine_01_capsule, center=<4.2817 0 0>}]
 thigh_l [{shapeType=Capsule, length=22, radius=7.5, shapeName=thigh_l_capsule, center=<-21 0 0>, rotation=<rot 90 90 90>}]
 thigh_r [{shapeType=Capsule, length=23, radius=7.5, shapeName=thigh_r_capsule, center=<21 0 0>, rotation=<rot 90 180 180>}]
}
constraints [
 {
  bone1Name=spine_01
  bone2Name=pelvis
  swing1LimitDegrees=45
  swing1Motion=Free
  swing2LimitDegrees=45
  swing2Motion=Free
  twistLimitDegrees=45
  twistMotion=Free
 }
 {
  bone1Name=thigh_l
  bone2Name=pelvis
  swing1LimitDegrees=45
  swing1Motion=Free
  swing2LimitDegrees=45
  swing2Motion=Free
  twistLimitDegrees=45
  twistMotion=Free
 }
 {
  bone1Name=thigh_r
  bone2Name=pelvis
  swing1LimitDegrees=45
  swing1Motion=Free
  swing2LimitDegrees=45
  swing2Motion=Free
  twistLimitDegrees=45
  twistMotion=Free
 }
 {bone1Name=dyn_jacket_fr_01_L, bone2Name=dyn_jacket_fr_MSTR_L, swing1LimitDegrees=2.5, swing2LimitDegrees=10, twistLimitDegrees=20}
 {bone1Name=dyn_jacket_fr_03_L, bone2Name=dyn_jacket_fr_02_L, swing1LimitDegrees=2.5, swing2LimitDegrees=45, twistLimitDegrees=5}
 {bone1Name=dyn_jacket_fr_02_L, bone2Name=dyn_jacket_fr_01_L, swing1LimitDegrees=2.5, swing2LimitDegrees=45, twistLimitDegrees=5}
 {bone1Name=dyn_jacket_01_R, bone2Name=dyn_jacket_MSTR_R, swing1LimitDegrees=5, swing2LimitDegrees=25, twistLimitDegrees=5}
 {bone1Name=dyn_jacket_02_R, bone2Name=dyn_jacket_01_R, swing1LimitDegrees=5, swing2LimitDegrees=25, twistLimitDegrees=5}
 {bone1Name=dyn_jacket_03_R, bone2Name=dyn_jacket_02_R, swing1LimitDegrees=5, swing2LimitDegrees=25, twistLimitDegrees=5}
 {bone1Name=dyn_jacket_bk_01, bone2Name=dyn_jacket_bk_MSTR, swing1LimitDegrees=5, swing2LimitDegrees=25, twistLimitDegrees=5}
 {bone1Name=dyn_jacket_bk_02, bone2Name=dyn_jacket_bk_01, swing1LimitDegrees=5, swing2LimitDegrees=25, twistLimitDegrees=5}
 {bone1Name=dyn_jacket_bk_03, bone2Name=dyn_jacket_bk_02, swing1LimitDegrees=5, swing2LimitDegrees=25, twistLimitDegrees=5}
 {bone1Name=dyn_jacket_bk_01_R, bone2Name=dyn_jacket_bk_MSTR_R, swing1LimitDegrees=5, swing2LimitDegrees=25, twistLimitDegrees=5}
 {bone1Name=dyn_jacket_bk_02_R, bone2Name=dyn_jacket_bk_01_R, swing1LimitDegrees=5, swing2LimitDegrees=25, twistLimitDegrees=5}
 {bone1Name=dyn_jacket_bk_03_R, bone2Name=dyn_jacket_bk_02_R, swing1LimitDegrees=5, swing2LimitDegrees=25, twistLimitDegrees=5}
 {bone1Name=dyn_jacket_fr_01_R, bone2Name=dyn_jacket_fr_MSTR_R, swing1LimitDegrees=5, swing2LimitDegrees=25, twistLimitDegrees=5}
 {bone1Name=dyn_jacket_fr_02_R, bone2Name=dyn_jacket_fr_01_R, swing1LimitDegrees=5, swing2LimitDegrees=25, twistLimitDegrees=5}
 {bone1Name=dyn_jacket_fr_03_R, bone2Name=dyn_jacket_fr_02_R, swing1LimitDegrees=5, swing2LimitDegrees=25, twistLimitDegrees=5}
 {bone1Name=dyn_jacket_01_L, bone2Name=dyn_jacket_MSTR_L, swing1LimitDegrees=2.5, swing2LimitDegrees=2.5, twistLimitDegrees=2.5}
 {bone1Name=dyn_jacket_02_L, bone2Name=dyn_jacket_01_L, swing1LimitDegrees=2.5, swing2LimitDegrees=45, twistLimitDegrees=5}
 {bone1Name=dyn_jacket_03_L, bone2Name=dyn_jacket_02_L, swing1LimitDegrees=2.5, swing2LimitDegrees=45, twistLimitDegrees=5}
 {bone1Name=dyn_jacket_bk_01_L, bone2Name=dyn_jacket_bk_MSTR_L, swing1LimitDegrees=2.5, swing2LimitDegrees=2.5, twistLimitDegrees=2.5}
 {bone1Name=dyn_jacket_bk_02_L, bone2Name=dyn_jacket_bk_01_L, swing1LimitDegrees=2.5, swing2LimitDegrees=45, twistLimitDegrees=5}
 {bone1Name=dyn_jacket_bk_03_L, bone2Name=dyn_jacket_bk_02_L, swing1LimitDegrees=5, swing2LimitDegrees=25, twistLimitDegrees=5}
]
