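# Physics asset for the scarf on the female medium SpeedBonny player body, set up for secondary motion of its dynamic bones.
# Three two-segment capsule chains hang from tiny MSTR anchor bodies: one front chain linked to spine_05, left and right back chains linked to neck_01.
# Anchor links rotate freely, while segment joints have swing limits and tight twist limits.
# Larger capsules on the upper spine, neck and clavicles approximate the body around the scarf.
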PhysicsAsset /Game/Characters/Player/Female/Medium/Bodies/F_MED_SpeedBonny/Meshes/F_MED_SpeedBonny_Scarf_Physics
bodies {
 clavicle_l [{shapeType=Capsule, length=6, radius=7, shapeName=clavicle_l_capsule, center=<7.1728 0 -1.2648>, rotation=<rot 80 -0.0007 -0.001>}]
 clavicle_r [{shapeType=Capsule, length=5.5, radius=7.5, shapeName=clavicle_r_capsule, center=<-6.4635 0 1.1397>, rotation=<rot 80 -0.0007 -0.001>}]
 dyn_scarf_bk_01_L [{shapeType=Capsule, length=5, radius=1.5, shapeName=dyn_scarf_bk_01_L_capsule, center=<4.9885 0 0>, rotation=<rot -90 180 179.9999>}]
 dyn_scarf_bk_01_R [{shapeType=Capsule, length=5, radius=1.5, shapeName=dyn_scarf_bk_01_R_capsule, center=<4.9885 0 0>, rotation=<rot -90 -179.9901 179.9902>}]
 dyn_scarf_bk_02_L [{shapeType=Capsule, length=5, radius=1.5, shapeName=dyn_scarf_bk_02_L_capsule, center=<6.3623 0 0>, rotation=<rot -90 160.5288 -160.5288>}]
 dyn_scarf_bk_02_R [{shapeType=Capsule, length=6, radius=1.5, shapeName=dyn_scarf_bk_02_R_capsule, center=<5.9732 0 -1.0467>, rotation=<rot -79.9998 180 180>}]
 dyn_scarf_bk_MSTR_L [{shapeType=Capsule, length=0, radius=0.5, shapeName=dyn_scarf_bk_MSTR_L_capsule}]
 dyn_scarf_bk_MSTR_R [{shapeType=Capsule, length=0, radius=0.5, shapeName=dyn_scarf_bk_MSTR_R_capsule}]
 dyn_scarf_fr_01 [{shapeType=Capsule, length=5, radius=1.5, shapeName=dyn_scarf_fr_01_capsule, center=<4.2012 0 0>, rotation=<rot -90 0 0>}]
 dyn_scarf_fr_02 [{shapeType=Capsule, length=5, radius=1.5, shapeName=dyn_scarf_fr_02_capsule, center=<4.2012 0 0>, rotation=<rot -90 0 0>}]
 dyn_scarf_fr_MSTR [{shapeType=Capsule, length=0, radius=0.5, shapeName=dyn_scarf_fr_MSTR_capsule}]
 neck_01 [{shapeType=Capsule, length=6.5, radius=4.5, shapeName=neck_01_capsule, center=<2.114 0 0>, rotation=<rot -90 160.5288 -160.529>}]
 spine_04 [{shapeType=Capsule, length=8.5, radius=12, shapeName=spine_04_capsule}]
 spine_05 [{shapeType=Capsule, length=10.5, radius=10.5, shapeName=spine_05_capsule}]
}
constraints [
 {
  bone1Name=dyn_scarf_fr_MSTR
  bone2Name=spine_05
  swing1LimitDegrees=45
  swing1Motion=Free
  swing2LimitDegrees=45
  swing2Motion=Free
  twistLimitDegrees=45
  twistMotion=Free
 }
 {
  bone1Name=spine_05
  bone2Name=spine_04
  swing1LimitDegrees=45
  swing1Motion=Free
  swing2LimitDegrees=45
  swing2Motion=Free
  twistLimitDegrees=45
  twistMotion=Free
 }
 {bone1Name=dyn_scarf_fr_01, bone2Name=dyn_scarf_fr_MSTR, swing1LimitDegrees=45, swing2LimitDegrees=70, twistLimitDegrees=5}
 {bone1Name=dyn_scarf_fr_02, bone2Name=dyn_scarf_fr_01, swing1LimitDegrees=50, swing2LimitDegrees=50, twistLimitDegrees=5}
 {bone1Name=dyn_scarf_bk_01_L, bone2Name=dyn_scarf_bk_MSTR_L, swing1LimitDegrees=45, swing2LimitDegrees=70, twistLimitDegrees=5}
 {bone1Name=dyn_scarf_bk_02_L, bone2Name=dyn_scarf_bk_01_L, swing1LimitDegrees=70, swing2LimitDegrees=70, twistLimitDegrees=5}
 {bone1Name=dyn_scarf_bk_01_R, bone2Name=dyn_scarf_bk_MSTR_R, swing1LimitDegrees=45, swing2LimitDegrees=70, twistLimitDegrees=5}
 {bone1Name=dyn_scarf_bk_02_R, bone2Name=dyn_scarf_bk_01_R, swing1LimitDegrees=70, swing2LimitDegrees=70, twistLimitDegrees=5}
 {
  bone1Name=dyn_scarf_bk_MSTR_L
  bone2Name=neck_01
  swing1LimitDegrees=45
  swing1Motion=Free
  swing2LimitDegrees=45
  swing2Motion=Free
  twistLimitDegrees=45
  twistMotion=Free
 }
 {
  bone1Name=dyn_scarf_bk_MSTR_R
  bone2Name=neck_01
  swing1LimitDegrees=45
  swing1Motion=Free
  swing2LimitDegrees=45
  swing2Motion=Free
  twistLimitDegrees=45
  twistMotion=Free
 }
]
