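# Post-process style AnimBlueprint for the SpeedBonny female cosmetic body, adding secondary motion to scarf, jacket, bags, belt and bracelets.
# Parent-mesh pose feeds Control Rig, corrective bones and scarf AnimDynamics through cached poses.
# RigidBody nodes then simulate with separate front-end, gameplay and bag assets, bypassed to the belt pose when disabled.
# EventGraph derives alphas from bone deltas and movement state, gravity overrides, and scarf/bag visibility from material parameters.
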AnimBlueprint /BRCosmetics/Characters/Player/Female/Medium/Bodies/F_MED_SpeedBonny/Meshes/F_MED_SpeedBonny_AnimBP
target_skeleton=&/F_MED_SpeedBonny_Skeleton
graphs [
 AnimGraph {nodes [
   RigidBody!(ComponentPose, Alpha, OverrideWorldGravity:Vector=Get Gravity Override(), ExternalForce) -> (Pose) {
    bOverrideWorldGravity=true
    overridePhysicsAsset=/Game/Characters/Player/Female/Medium/Bodies/F_MED_SpeedBonny/Meshes/F_MED_SpeedBonny_Scarf_Physics
    simulationSpace=BaseBoneSpace
    alphaBoolBlend {blendOption=Linear}
    baseBoneRef {boneName=spine_05}
    componentAppliedLinearAccClamp=<3000 3000 3000>
    componentLinearAccScale=<0.6 0.6 0.6>
    componentLinearVelScale=<0.6 0.6 0.6>
   }
   n1 = Get Gravity Override()
   n2 = RigidBody!(
    ComponentPose=Local To Component(LocalPose=Use cached pose 'dyn cache'())
    Alpha=Get is in FrontEnd() -> (is in FrontEnd:double)
    OverrideWorldGravity=@n1
   ) -> (Pose) {
    bOverrideWorldGravity=true
    overridePhysicsAsset=&/F_MED_SpeedBonny_FrontEnd_Physics
    simulationSpace=BaseBoneSpace
    alphaBoolBlend {blendOption=Linear}
    baseBoneRef {boneName=pelvis}
    componentAppliedLinearAccClamp=<4000 4000 4000>
    componentLinearAccScale=<0.35 0.35 0.35>
    componentLinearVelScale=<0.35 0.35 0.35>
   }
   n3 = RigidBody!(
    ComponentPose=@n2
    Alpha=Get is not in FrontEnd() -> (is not in FrontEnd:double)
    OverrideWorldGravity=@n1
    ExternalForce:Vector
    ComponentLinearAccScale:Vector=Get ComponentLinearAccScale() -> (ComponentLinearAccScale:Vector)
    ComponentLinearVelScale:Vector=Get ComponentLinearVelScale() -> (ComponentLinearVelScale:Vector)
   ) -> (Pose) {
    bOverrideWorldGravity=true
    overridePhysicsAsset=&/F_MED_SpeedBonny__Physics
    simulationSpace=BaseBoneSpace
    alphaBoolBlend {blendOption=Linear}
    baseBoneRef {boneName=pelvis}
    componentAppliedLinearAccClamp=<4000 4000 4000>
    componentLinearAccScale=<0.35 0.35 0.35>
    componentLinearVelScale=<0.35 0.35 0.35>
   }
   n4 = Component To Local(ComponentPose=@n3)
   n5 = "Transform (Modify) Bone - Bone: dyn_bags_MSTR"(
    ComponentPose=Local To Component(LocalPose=Use cached pose 'dyn cache'())
    Alpha="Get Bags correction when not crouching "() -> ("Bags correction when not crouching ":double)
    Translation="0,0,0"
    Rotation="50,0,0"
    Scale="1,1,1"
   ) {
    rotationMode=BMM_Additive
    rotationSpace=BCS_BoneSpace
    translationMode=BMM_Additive
    translationSpace=BCS_BoneSpace
    alphaBoolBlend {blendOption=Linear}
    alphaScaleBiasClamp {bInterpResult=true, interpSpeedDecreasing=20, interpSpeedIncreasing=20}
    boneToModify {boneName=dyn_bags_MSTR}
    rotation {pitch=50}
   }
   n6 = "Transform (Modify) Bone - Bone: dyn_bags_MSTR"(
    ComponentPose=@n5
    Alpha=Get Bags correction when crouching standing() -> (Bags correction when crouching standing:double)
    Translation="-8.108,1.145,-0.629"
    Rotation="20.148,0.131,-1.552"
    Scale="1,1,1"
   ) {
    rotationMode=BMM_Additive
    rotationSpace=BCS_BoneSpace
    translationMode=BMM_Additive
    translationSpace=BCS_BoneSpace
    alphaBoolBlend {blendOption=Linear}
    alphaScaleBiasClamp {bInterpResult=true, interpSpeedIncreasing=18}
    boneToModify {boneName=dyn_bags_MSTR}
    rotation=<rot 20.148 0.131 -1.552>
    translation=<-8.108 1.145 -0.629>
   }
   n7 = "Transform (Modify) Bone - Bone: dyn_bags_MSTR"(
    ComponentPose=@n6
    Alpha="Get Bags correction when crouching "() -> ("Bags correction when crouching ":double)
    Translation="-4.128,0.296,0.983"
    Rotation="20,0,0"
    Scale="1,1,1"
   ) {
    rotationMode=BMM_Additive
    rotationSpace=BCS_BoneSpace
    translationMode=BMM_Additive
    translationSpace=BCS_BoneSpace
    alphaBoolBlend {blendOption=Linear}
    alphaScaleBiasClamp {bInterpResult=true, interpSpeedDecreasing=20, interpSpeedIncreasing=20}
    boneToModify {boneName=dyn_bags_MSTR}
    rotation {pitch=20}
    translation=<-4.128 0.296 0.983>
   }
   n8 = Get Gravity Override()
   n9 = RigidBody!(ComponentPose=@n7, Alpha=Get is in FrontEnd(), OverrideWorldGravity=@n8) -> (Pose) {
    bOverrideWorldGravity=true
    overridePhysicsAsset=&/F_MED_SpeedBonny_Bags_FrontEnd_Physics
    simulationSpace=BaseBoneSpace
    alphaBoolBlend {blendOption=Linear}
    baseBoneRef {boneName=pelvis}
    componentAppliedLinearAccClamp=<4000 4000 4000>
    componentLinearAccScale=<0.35 0.35 0.35>
    componentLinearVelScale=<0.35 0.35 0.35>
   }
   n10 = RigidBody!(
    ComponentPose=@n9
    Alpha=Get is not in FrontEnd()
    OverrideWorldGravity=@n8
    ExternalForce:Vector
    ComponentLinearAccScale:Vector=Get ComponentLinearAccScale()
    ComponentLinearVelScale:Vector=Get ComponentLinearVelScale()
   ) -> (Pose) {
    bOverrideWorldGravity=true
    overridePhysicsAsset=&/F_MED_SpeedBonny_Bags_Physics
    simulationSpace=BaseBoneSpace
    alphaBoolBlend {blendOption=Linear}
    baseBoneRef {boneName=pelvis}
    componentAppliedLinearAccClamp=<4000 4000 4000>
    componentLinearAccScale=<0.35 0.35 0.35>
    componentLinearVelScale=<0.35 0.35 0.35>
   }
   n11 = Component To Local(ComponentPose=@n10)
   n12 = Get hidden bags() -> (hidden bags:bool)
   Output Pose(Result=Blend Poses by bool(
     BlendPose_0=Blend Poses by bool(BlendPose_0=@n4, BlendPose_1=@n11, bActiveValue=@n12) {blendType=Linear}
     BlendPose_1=Use cached pose 'belt cache'()
     bActiveValue=Get bRigidBodyEnabled()
    ) {blendType=Linear}) {layerGroup=None}
   n13 = "Transform (Modify) Bone - Bone: belt_fr"(
    ComponentPose=Local To Component(LocalPose=Use cached pose 'main cahe'())
    Alpha=Get Upperarm R Front() -> (Upperarm R Front:float)
    Translation="2.903,-2.366,-1.038"
    Rotation="0,0,-10"
    Scale="1,1,1"
   ) {
    rotationMode=BMM_Additive
    rotationSpace=BCS_BoneSpace
    scaleSpace=BCS_BoneSpace
    translationMode=BMM_Additive
    translationSpace=BCS_BoneSpace
    alphaBoolBlend {blendOption=Linear}
    boneToModify {boneName=belt_fr}
    rotation {roll=-10}
    translation=<2.903 -2.366 -1.038>
   }
   n14 = Get Clavicle R Back() -> (Clavicle R Back:float)
   n15 = "Transform (Modify) Bone - Bone: belt_bk"(ComponentPose=@n13, Alpha=@n14, Translation="-0.618,2.933,-2.577", Rotation="-0.000,-10.000,-0.000", Scale="1,1,1") {
    rotationMode=BMM_Additive
    rotationSpace=BCS_BoneSpace
    scaleSpace=BCS_BoneSpace
    translationMode=BMM_Additive
    translationSpace=BCS_BoneSpace
    alphaBoolBlend {blendOption=Linear}
    boneToModify {boneName=belt_bk}
    rotation=<rot -0.0001 -10 -0.0001>
    translation=<-0.618 2.933 -2.577>
   }
   n16 = Get Clavicle R up() -> (Clavicle R up:float)
   n17 = "Transform (Modify) Bone - Bone: belt_fr"(ComponentPose=@n15, Alpha=@n16, Translation="-0.000,0.000,-2.486", Rotation="0,0,0", Scale="1,1,1") {
    rotationMode=BMM_Additive
    rotationSpace=BCS_BoneSpace
    scaleSpace=BCS_BoneSpace
    translationMode=BMM_Additive
    translationSpace=BCS_BoneSpace
    alphaBoolBlend {blendOption=Linear}
    boneToModify {boneName=belt_fr}
    rotation=<rot 0 0 0>
    translation {z=-2.486}
   }
   n18 = "Transform (Modify) Bone - Bone: belt_spine"(ComponentPose=@n17, Alpha=@n14, Translation="-0.000,-4.459,-3.553", Rotation="0,0,0", Scale="1,1,1") {
    rotationMode=BMM_Additive
    rotationSpace=BCS_BoneSpace
    scaleSpace=BCS_BoneSpace
    translationMode=BMM_Additive
    translationSpace=BCS_BoneSpace
    alphaBoolBlend {blendOption=Linear}
    boneToModify {boneName=belt_spine}
    rotation=<rot 0 -0.0001 0>
    translation {y=-4.459, z=-3.553}
   }
   n19 = "Transform (Modify) Bone - Bone: belt_bk"(ComponentPose=@n18, Alpha=@n16, Translation="0.000,2.706,-2.724", Rotation="0,0,0", Scale="1,1,1") {
    rotationMode=BMM_Additive
    rotationSpace=BCS_BoneSpace
    scaleSpace=BCS_BoneSpace
    translationMode=BMM_Additive
    translationSpace=BCS_BoneSpace
    alphaBoolBlend {blendOption=Linear}
    boneToModify {boneName=belt_bk}
    translation {y=2.706, z=-2.724}
   }
   Save cached pose 'belt cache'(Pose=Component To Local(ComponentPose=@n19)) {cacheName=belt cache}
   n20 = "Get Thigh l bending "() -> ("Thigh l bending ":float)
   n21 = Get Gravity Override()
   n22 = Get hidden scarf() -> (hidden scarf:double)
   n23 = Use cached pose 'main cahe'()
   n24 = Use cached pose 'belt cache'()
   n25 = Get hidden bags()
   n26 = "Constraint - dyn_scarf_bk_MSTR_L (PPPPPPPP)"!(
    ComponentPose=Local To Component(LocalPose=Blend Poses by bool(BlendPose_0=@n23, BlendPose_1=@n24, bActiveValue=@n25) {blendType=Linear})
    Alpha=@n22
    ConstraintWeights_0=0.0583
    ConstraintWeights_1:float=0.1964
    ConstraintWeights_2:float=0.0232
    ConstraintWeights_3:float=0.0066
    ConstraintWeights_4:float=0.4444
    ConstraintWeights_5:float=0.1213
    ConstraintWeights_6:float=0.1478
    ConstraintWeights_7:float=0.002
   ) -> (Pose) {
    alphaBoolBlend {blendOption=Linear}
    boneToModify {boneName=dyn_scarf_bk_MSTR_L}
    constraintSetup [
     {targetBone {boneName=spine_04}}
     {targetBone {boneName=spine_05}}
     {targetBone {boneName=clavicle_l}}
     {targetBone {boneName=clavicle_r}}
     {targetBone {boneName=neck_01}}
     {targetBone {boneName=head}}
     {targetBone {boneName=neck_02}}
     {targetBone {boneName=dyn_scarf_bk_01_L}}
    ]
    constraintWeights [1,1,1,1,1,1,1,1]
   }
   n27 = "Constraint - dyn_scarf_bk_MSTR_R (PPPPPPPP)"!(
    ComponentPose=@n26
    Alpha=@n22
    ConstraintWeights_0=0.057
    ConstraintWeights_1:float=0.2253
    ConstraintWeights_2:float=0.007
    ConstraintWeights_3:float=0.0284
    ConstraintWeights_4:float=0.4188
    ConstraintWeights_5:float=0.1069
    ConstraintWeights_6:float=0.1357
    ConstraintWeights_7:float=0.0209
   ) -> (Pose) {
    alphaBoolBlend {blendOption=Linear}
    boneToModify {boneName=dyn_scarf_bk_MSTR_R}
    constraintSetup [
     {targetBone {boneName=spine_04}}
     {targetBone {boneName=spine_05}}
     {targetBone {boneName=clavicle_l}}
     {targetBone {boneName=clavicle_r}}
     {targetBone {boneName=neck_01}}
     {targetBone {boneName=head}}
     {targetBone {boneName=neck_02}}
     {targetBone {boneName=dyn_scarf_bk_01_R}}
    ]
    constraintWeights [1,1,1,1,1,1,1,1]
   }
   n28 = Anim Dynamics - Bone: dyn_scarf_bk_01_L!(ComponentPose=@n27, Alpha=@n22, GravityOverride:Vector=@n21, ExternalForce="0,0,0") -> (Pose) {
    angularBiasOverride=1
    angularDampingOverride=1
    angularSpringConstant=1500
    bAngularSpring=true
    bOverrideAngularBias=true
    bOverrideAngularDamping=true
    bUseGravityOverride=true
    bUsePlanarLimit=false
    bUseSphericalLimits=true
    simulationSpace=RootRelative
    alphaBoolBlend {blendOption=Linear}
    boundBone {boneName=dyn_scarf_bk_01_L}
    componentAppliedLinearAccClamp=<5000 5000 5000>
    componentLinearAccScale=<0.9 0.9 0.7>
    componentLinearVelScale=<0.9 0.9 0.7>
    physicsBodyDefinitions [{
      collisionType=CustomSphere
      sphereCollisionRadius=1
      boundBone {boneName=dyn_scarf_bk_01_L}
      boxExtents=<50 15 15>
      constraintSetup {angularLimitsMax=<0 45 20>, angularLimitsMin=<0 -45 -20>, angularTarget=<0 0 0>}
      localJointOffset=<8 0 0>
     }]
    sphericalLimits [
     {limitRadius=11.5, drivingBone {boneName=spine_05}, sphereLocalOffset=<0.9612 1.924 -1.1508>}
     {limitRadius=10.5, drivingBone {boneName=clavicle_l}, sphereLocalOffset=<6.7739 -1.1084 -3.6814>}
     {limitRadius=10, drivingBone {boneName=clavicle_r}, sphereLocalOffset=<-4.6819 2.6135 2.9389>}
    ]
   }
   n29 = Anim Dynamics - Bone: dyn_scarf_bk_02_L!(ComponentPose=@n28, Alpha=@n22, GravityOverride:Vector=@n21, ExternalForce="0,0,0") -> (Pose) {
    angularBiasOverride=1
    angularDampingOverride=1
    angularSpringConstant=1500
    bAngularSpring=true
    bOverrideAngularBias=true
    bOverrideAngularDamping=true
    bUseGravityOverride=true
    bUsePlanarLimit=false
    bUseSphericalLimits=true
    simulationSpace=RootRelative
    alphaBoolBlend {blendOption=Linear}
    boundBone {boneName=dyn_scarf_bk_02_L}
    componentAppliedLinearAccClamp=<5000 5000 5000>
    componentLinearAccScale=<0.9 0.9 0.7>
    componentLinearVelScale=<0.9 0.9 0.7>
    physicsBodyDefinitions [{
      collisionType=CustomSphere
      sphereCollisionRadius=1
      boundBone {boneName=dyn_scarf_bk_02_L}
      boxExtents=<40 10 10>
      constraintSetup {angularLimitsMax=<0 55 45>, angularLimitsMin=<0 -55 -45>, angularTarget=<0 0 0>}
      localJointOffset=<8 0 0>
     }]
    sphericalLimits [
     {limitRadius=11.5, drivingBone {boneName=spine_05}, sphereLocalOffset=<2.279 0.4028 0>}
     {limitRadius=10.5, drivingBone {boneName=clavicle_l}, sphereLocalOffset=<6.7739 -1.1084 -3.6814>}
     {limitRadius=10, drivingBone {boneName=clavicle_r}, sphereLocalOffset=<-4.6819 2.6135 2.9389>}
    ]
   }
   n30 = Anim Dynamics - Bone: dyn_scarf_bk_01_R!(ComponentPose=@n29, Alpha=@n22, GravityOverride:Vector=@n21, ExternalForce="0,0,0") -> (Pose) {
    angularBiasOverride=1
    angularDampingOverride=1
    angularSpringConstant=1500
    bAngularSpring=true
    bOverrideAngularBias=true
    bOverrideAngularDamping=true
    bUseGravityOverride=true
    bUsePlanarLimit=false
    bUseSphericalLimits=true
    simulationSpace=RootRelative
    alphaBoolBlend {blendOption=Linear}
    boundBone {boneName=dyn_scarf_bk_01_R}
    componentAppliedLinearAccClamp=<5000 5000 5000>
    componentLinearAccScale=<0.9 0.9 0.7>
    componentLinearVelScale=<0.9 0.9 0.7>
    physicsBodyDefinitions [{
      collisionType=CustomSphere
      sphereCollisionRadius=1
      boundBone {boneName=dyn_scarf_bk_01_R}
      boxExtents=<50 15 15>
      constraintSetup {angularLimitsMax=<0 45 20>, angularLimitsMin=<0 -45 -20>, angularTarget=<0 0 0>}
      localJointOffset=<8 0 0>
     }]
    sphericalLimits [
     {limitRadius=13.5, drivingBone {boneName=spine_05}, sphereLocalOffset=<3.4126 3.582 -1.0984>}
     {limitRadius=10.5, drivingBone {boneName=clavicle_l}, sphereLocalOffset=<6.7739 -1.1084 -3.6814>}
     {limitRadius=10.5, drivingBone {boneName=clavicle_r}, sphereLocalOffset=<-4.0249 -0.6315 3.2228>}
    ]
   }
   n31 = Anim Dynamics - Bone: dyn_scarf_bk_02_R!(ComponentPose=@n30, Alpha=@n22, GravityOverride:Vector=@n21, ExternalForce="0,0,0") -> (Pose) {
    angularBiasOverride=1
    angularDampingOverride=1
    angularSpringConstant=1500
    bAngularSpring=true
    bOverrideAngularBias=true
    bOverrideAngularDamping=true
    bUseGravityOverride=true
    bUsePlanarLimit=false
    bUseSphericalLimits=true
    simulationSpace=RootRelative
    alphaBoolBlend {blendOption=Linear}
    boundBone {boneName=dyn_scarf_bk_02_R}
    componentAppliedLinearAccClamp=<5000 5000 5000>
    componentLinearAccScale=<0.9 0.9 0.7>
    componentLinearVelScale=<0.9 0.9 0.7>
    physicsBodyDefinitions [{
      collisionType=CustomSphere
      sphereCollisionRadius=1
      boundBone {boneName=dyn_scarf_bk_02_R}
      boxExtents=<40 10 10>
      constraintSetup {angularLimitsMax=<0 55 45>, angularLimitsMin=<0 -55 -45>, angularTarget=<0 0 0>}
      localJointOffset=<8 0 0>
     }]
    sphericalLimits [
     {limitRadius=11.5, drivingBone {boneName=spine_05}, sphereLocalOffset=<3.5539 -0.8374 -1.0848>}
     {limitRadius=10.5, drivingBone {boneName=clavicle_l}, sphereLocalOffset=<6.7739 -1.1084 -3.6814>}
     {limitRadius=12, drivingBone {boneName=clavicle_r}, sphereLocalOffset=<-2.3157 0.4217 1.4509>}
     {limitRadius=13, drivingBone {boneName=spine_05}, sphereLocalOffset=<0 0 7.7804>}
    ]
   }
   n32 = Anim Dynamics - Bone: dyn_scarf_fr_01!(ComponentPose=@n31, Alpha=@n22, GravityOverride:Vector=@n21, ExternalForce="0,0,0") -> (Pose) {
    angularBiasOverride=1
    angularDampingOverride=1
    angularSpringConstant=1500
    bAngularSpring=true
    bOverrideAngularBias=true
    bOverrideAngularDamping=true
    bUseGravityOverride=true
    bUsePlanarLimit=false
    bUseSphericalLimits=true
    simulationSpace=RootRelative
    alphaBoolBlend {blendOption=Linear}
    boundBone {boneName=dyn_scarf_fr_01}
    componentAppliedLinearAccClamp=<5000 5000 5000>
    componentLinearAccScale=<0.9 0.9 0.7>
    componentLinearVelScale=<0.9 0.9 0.7>
    physicsBodyDefinitions [{
      collisionType=CustomSphere
      sphereCollisionRadius=1
      boundBone {boneName=dyn_scarf_fr_01}
      boxExtents=<50 20 20>
      constraintSetup {angularLimitsMax=<0 45 20>, angularLimitsMin=<0 -45 -20>, angularTarget=<0 0 0>}
      localJointOffset=<8 0 0>
     }]
    sphericalLimits [
     {limitRadius=12, drivingBone {boneName=spine_05}, sphereLocalOffset=<0 0 0>}
     {limitRadius=11, drivingBone {boneName=spine_04}, sphereLocalOffset=<0 4.7391 -0.3201>}
    ]
   }
   n33 = Anim Dynamics - Bone: dyn_scarf_fr_02!(ComponentPose=@n32, Alpha=@n22, GravityOverride:Vector=@n21, ExternalForce="0,0,0") -> (Pose) {
    angularBiasOverride=1
    angularDampingOverride=1
    angularSpringConstant=1500
    bAngularSpring=true
    bOverrideAngularBias=true
    bOverrideAngularDamping=true
    bUseGravityOverride=true
    bUsePlanarLimit=false
    bUseSphericalLimits=true
    simulationSpace=RootRelative
    alphaBoolBlend {blendOption=Linear}
    boundBone {boneName=dyn_scarf_fr_02}
    componentAppliedLinearAccClamp=<5000 5000 5000>
    componentLinearAccScale=<0.9 0.9 0.7>
    componentLinearVelScale=<0.9 0.9 0.7>
    physicsBodyDefinitions [{
      collisionType=CustomSphere
      sphereCollisionRadius=1
      boundBone {boneName=dyn_scarf_fr_02}
      boxExtents=<45 20 20>
      constraintSetup {angularLimitsMax=<0 55 45>, angularLimitsMin=<0 -55 -45>, angularTarget=<0 0 0>}
      localJointOffset=<8 0 0>
     }]
    sphericalLimits [
     {limitRadius=12, drivingBone {boneName=spine_05}, sphereLocalOffset=<0 0 0>}
     {limitRadius=11, drivingBone {boneName=spine_04}, sphereLocalOffset=<0 2.7587 -0.435>}
    ]
   }
   n34 = "Constraint - dyn_jacket_fr_CST_L (PP)"!(ComponentPose=@n33, Alpha=@n20, ConstraintWeights_0=0.5, ConstraintWeights_1:float=0.5) -> (Pose) {
    alphaBoolBlend {blendOption=Linear}
    boneToModify {boneName=dyn_jacket_fr_CST_L}
    constraintSetup [{targetBone {boneName=pelvis}}, {targetBone {boneName=spine_02}}]
    constraintWeights [1,1]
   }
   n35 = "Get Thigh l bending "()
   n36 = "Transform (Modify) Bone - Bone: dyn_jacket_fr_MSTR_L"(
    ComponentPose=@n34
    Alpha=@n35
    Translation="0,0,0"
    Rotation=Get Jacket front l rotation() -> (Jacket front l rotation:Rotator)
    Scale="1,1,1"
   ) {
    rotationMode=BMM_Additive
    rotationSpace=BCS_BoneSpace
    translationMode=BMM_Additive
    translationSpace=BCS_BoneSpace
    alphaBoolBlend {blendOption=Linear}
    boneToModify {boneName=dyn_jacket_fr_MSTR_L}
    rotation=<rot 30.0122 -23.6819 -90.1671>
   }
   n37 = Get Thigh r bending() -> (Thigh r bending:float)
   n38 = "Transform (Modify) Bone - Bone: dyn_jacket_fr_MSTR_R"(ComponentPose=@n36, Alpha=@n37, Translation="0,0,0", Rotation="40,5,20", Scale="1,1,1") {
    rotationMode=BMM_Additive
    rotationSpace=BCS_BoneSpace
    translationMode=BMM_Additive
    translationSpace=BCS_BoneSpace
    alphaBoolBlend {blendOption=Linear}
    boneToModify {boneName=dyn_jacket_fr_MSTR_R}
    rotation=<rot 40 5 20>
   }
   n39 = "Constraint - dyn_jacket_CST_L (PP)"!(ComponentPose=@n38, Alpha=@n20, ConstraintWeights_0=0.5, ConstraintWeights_1:float=0.5) -> (Pose) {
    alphaBoolBlend {blendOption=Linear}
    boneToModify {boneName=dyn_jacket_CST_L}
    constraintSetup [{targetBone {boneName=pelvis}}, {targetBone {boneName=spine_02}}]
    constraintWeights [1,1]
   }
   n40 = "Transform (Modify) Bone - Bone: dyn_jacket_MSTR_L"(ComponentPose=@n39, Alpha=@n35, Translation="0,0,0", Rotation="7,0,0", Scale="1,1,1") {
    rotationMode=BMM_Additive
    rotationSpace=BCS_BoneSpace
    translationMode=BMM_Additive
    translationSpace=BCS_BoneSpace
    alphaBoolBlend {blendOption=Linear}
    boneToModify {boneName=dyn_jacket_MSTR_L}
    rotation {pitch=7}
   }
   n41 = "Transform (Modify) Bone - Bone: dyn_jacket_MSTR_R"(ComponentPose=@n40, Alpha=@n37, Translation="0,0,0", Rotation="20,0,0", Scale="1,1,1") {
    rotationMode=BMM_Additive
    rotationSpace=BCS_BoneSpace
    translationMode=BMM_Additive
    translationSpace=BCS_BoneSpace
    alphaBoolBlend {blendOption=Linear}
    boneToModify {boneName=dyn_jacket_MSTR_R}
    rotation {pitch=20}
   }
   Save cached pose 'dyn cache'(Pose=Component To Local(ComponentPose=@n41)) {cacheName=dyn cache}
   n42 = Control Rig(Source=Copy Pose From Mesh() {bUseAttachedParent=true}) {
    alphaBoolBlend {blendOption=Linear}
    controlRigAssetReference {blueprintRigClass=&/F_MED_SpeedBonny_ControlRig.F_MED_SpeedBonny_ControlRig_C}
   }
   n43 = "Transform (Modify) Bone - Bone: bracelet_L"(
    ComponentPose=Local To Component(LocalPose=@n42)
    Alpha=Get Bracelet L Correction() -> (Bracelet L Correction:double)
    Translation="-1.5,0,0"
    Rotation="0,0,0"
    Scale="1,1,1"
   ) {
    translationMode=BMM_Additive
    translationSpace=BCS_BoneSpace
    alphaBoolBlend {blendOption=Linear}
    boneToModify {boneName=bracelet_L}
    translation {x=-1.5}
   }
   n44 = "Transform (Modify) Bone - Bone: bracelet_R"(
    ComponentPose=@n43
    Alpha="Get Bracelet  R Correction"() -> ("Bracelet  R Correction":float)
    Translation="1.5,0,0"
    Rotation="0,0,0"
    Scale="1,1,1"
   ) {
    translationMode=BMM_Additive
    translationSpace=BCS_BoneSpace
    alphaBoolBlend {blendOption=Linear}
    boneToModify {boneName=bracelet_R}
    translation {x=1.5}
   }
   Save cached pose 'main cahe'(Pose=Component To Local(ComponentPose=@n44)) {cacheName=main cahe}
  ]}
 EventGraph {nodes [
   Event BlueprintUpdateAnimation()
   > n1 = Sequence!(execute) -> (then_0, then_1, then_2:Exec)
   > n2 = Get Dynamic Material Instance(SkeletalMeshComponent=GetPartSkeletalMeshComponent())
   > n3 = Get Scalar Parameter Value!(execute, self:Material Instance Dynamic Object Reference=@n2, ParameterName:Name=Hide Element 02) -> (then, ReturnValue:float) {is=title K2Node_CallFunction}
   > Set hidden bags(hidden bags:bool="Equal (Float)"(A=@n3, B=1)) -> (Output_Get:bool)
   > n4 = Get Scalar Parameter Value(self=@n2, ParameterName=Hide Element 01)
   > Set hidden scarf(hidden scarf:double=MapRangeClamped(Value=@n4, InRangeA=0, InRangeB=1, OutRangeA=1, OutRangeB=0)) -> (Output_Get:double)
   n5 = Gravity Override State Function DEPRECATED(
    execute=@n1.then_1
    CharacterPartAnimInstance=Self-Reference()
    GameTimeInSeconds=GetGameTimeInSeconds()
    GravityOverrideParams=Get Gravity Override Params()
   )
   > Set Gravity Override(Gravity Override=@n5)
   n6 = SelectVector(A="0.1,0.1,0.1", B="0.35,0.35,0.35", bPickA=Get bIsSkyDiving())
   > Set ComponentLinearAccScale(ComponentLinearAccScale:Vector=@n6) -> (Output_Get:Vector)
   > Set ComponentLinearVelScale(ComponentLinearVelScale:Vector=@n6) -> (Output_Get:Vector)
   > Set Jacket front l rotation(Jacket front l rotation:Rotator=SelectRotator!(A:Rotator="15,-10,-50", B:Rotator="30.012,-23.682,-90.167", bPickA:bool=Get bIsSprinting()) -> (ReturnValue:Rotator)) -> (Output_Get:Rotator)
   > n7 = Branch(Condition=Get bIsInFrontEnd())
   > Set is in FrontEnd(is in FrontEnd:double=1) -> (Output_Get:double)
   > Set is not in FrontEnd(is not in FrontEnd:double=0) -> (Output_Get:double)
   Set is in FrontEnd(execute=@n7.else, is in FrontEnd=0)
   > Set is not in FrontEnd(is not in FrontEnd=1)
   n8 = Get SkelMesh()
   "Set Thigh l bending "(
    execute=@n1.then_2
    "Thigh l bending ":float=GetAlphaFromReferenceTransformDelta(SkeletalMeshComponent=@n8, BoneName=thigh_l, BaseName=pelvis, InputRange=(X=-25,Y=-100), TwistAxis="1,0,0")
   ) -> (Output_Get:float)
   n9 = GetAlphaFromReferenceTransformDelta(SkeletalMeshComponent=@n8, BoneName=thigh_l, BaseName=pelvis, InputRange=(X=-50,Y=-100), TwistAxis="1,0,0")
   n10 = NOT Boolean(A=Get bIsMoving2D())
   n11 = Get bIsCrouching()
   > Set Bags correction when crouching standing(Bags correction when crouching standing:double=SelectFloat(A=@n9, B=0, bPickA=AND Boolean(A=@n11, B=@n10))) -> (Output_Get:double)
   n12 = NOT Boolean(A=Get bIsSliding())
   n13 = Get bIsMoving2D()
   n14 = Get bIsCrouching()
   > "Set Bags correction when crouching "("Bags correction when crouching ":double=SelectFloat(A=1, B=0, bPickA=AND Boolean!(A=@n14, B=@n13, C:bool=@n12) -> (ReturnValue))) -> (Output_Get:double)
   n15 = OR Boolean(A=Get bIsCrouching(), B=Get bIsDBNO())
   > "Set Bags correction when not crouching "("Bags correction when not crouching ":double=SelectFloat(A=@n9, B=0, bPickA=NOT Boolean(A=@n15))) -> (Output_Get:double)
   > Set Thigh r bending(Thigh r bending:float=GetAlphaFromReferenceTransformDelta(SkeletalMeshComponent=@n8, BoneName=thigh_r, BaseName=pelvis, InputRange=(X=0,Y=-100), TwistAxis="1,0,0")) -> (Output_Get:float)
   n16 = Get SkelMesh()
   > Set Bracelet L Correction(Bracelet L Correction:double=float + float(
     A=GetAlphaFromReferenceTransformDelta(SkeletalMeshComponent=@n16, BoneName=hand_l, BaseName=lowerarm_l, DeltaComponentType=EulerY, InputRange=(X=0,Y=30), TwistAxis="1,0,0")
     B=GetAlphaFromReferenceTransformDelta(SkeletalMeshComponent=@n16, BoneName=hand_l, BaseName=lowerarm_l, DeltaComponentType=EulerY, InputRange=(X=0,Y=-30), TwistAxis="1,0,0")
    )) -> (Output_Get:double)
   > "Set Bracelet  R Correction"("Bracelet  R Correction":float=float + float(
     A=GetAlphaFromReferenceTransformDelta(SkeletalMeshComponent=@n16, BoneName=hand_r, BaseName=lowerarm_r, DeltaComponentType=EulerY, InputRange=(X=0,Y=30), TwistAxis="1,0,0")
     B=GetAlphaFromReferenceTransformDelta(
      SkeletalMeshComponent=@n16
      BoneName=hand_r
      BaseName=lowerarm_r
      DeltaComponentType=EulerY
      InputRange=(X=0,Y=-30)
      OutputRange=(X=0,Y=0.4)
      TwistAxis="1,0,0"
     )
    )) -> (Output_Get:float)
   > Set Clavicle R Back(Clavicle R Back:float=float + float(
     A=GetAlphaFromReferenceTransformDelta(
      SkeletalMeshComponent=@n16
      BoneName=spine_05
      BaseName=spine_04
      DeltaComponentType=EulerX
      InputRange=(X=0,Y=-40)
      OutputRange=(X=0,Y=0.5)
      TwistAxis="1,0,0"
     )
     B=GetAlphaFromReferenceTransformDelta(SkeletalMeshComponent=@n16, BoneName=clavicle_r, BaseName=spine_05, InputRange=(X=-5,Y=-22), OutputRange=(X=0,Y=0.5), TwistAxis="1,0,0")
    )) -> (Output_Get:float)
   > Set Clavicle R up(Clavicle R up:float=GetAlphaFromReferenceTransformDelta(SkeletalMeshComponent=@n16, BoneName=clavicle_r, BaseName=spine_05, DeltaComponentType=EulerY, InputRange=(X=10,Y=47), TwistAxis="1,0,0")) -> (Output_Get:float)
   > Set Upperarm R Front(Upperarm R Front:float=GetAlphaFromReferenceTransformDelta(SkeletalMeshComponent=@n16, BoneName=upperarm_r, BaseName=clavicle_r, InputRange=(X=0,Y=85), TwistAxis="1,0,0")) -> (Output_Get:float)
   Event BlueprintInitializeAnimation()
   > Set SkelMesh(SkelMesh=GetOwnerSkeletalMeshComponent())
   PrintString!(execute, InString=Hello, bPrintToScreen, bPrintToLog, TextColor, Duration=10000, Key) -> (then)
   GetAlphaFromReferenceTransformDelta(SkeletalMeshComponent=@n16, BoneName=clavicle_r, BaseName=spine_05, InputRange=(X=-10,Y=-25), TwistAxis="1,0,0")
   BreakVector(InVec=Get PawnVelocity())
  ]}
]
variables [
 "Bags correction when crouching " {type=number}
 Bags correction when crouching standing {type=number}
 "Bags correction when not crouching " {type=number}
 "Bracelet  R Correction" {type=number}
 Bracelet L Correction {type=number}
 Clavicle R Back {type=number}
 Clavicle R up {type=number}
 ComponentLinearAccScale {type=Vector}
 ComponentLinearVelScale {type=Vector}
 Gravity Override {type=Vector}
 Gravity Override Params {
  type=GravityOverrideParamsStruct
  default {
   skydive_FloatingDown {gravityOverride=<0 -1200 100>}
   skydive_GoingUp {gravityOverride=<0 -1200 100>}
   skydive_Parachuting {windFrequency=<1.5 2 8>}
  }
 }
 hidden bags {type=boolean}
 hidden scarf {type=number}
 is in FrontEnd {type=number}
 is not in FrontEnd {type=number}
 Jacket front l rotation {type=Rotator}
 SkelMesh {type=SkeletalMeshComponent}
 "Thigh l bending " {type=number}
 Thigh r bending {type=number}
 Upperarm R Front {type=number}
]
