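# Control Rig for the SpeedBonny character's face accessory that adjusts dynamic strand bones according to head pose.
# A spherical pose reader measures the head against spine_05, with an asymmetric falloff region.
# Its output weights a Modify Transforms node that applies preset rotations to strand master bones, including back and side2.
# A "Set Transforms for Bone" additive helper function and a get_head_look_up_alpha variable exist but are unused in the main graph.
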ControlRigBlueprint /BRCosmetics/Characters/Player/Female/Medium/Bodies/F_MED_SpeedBonny/Meshes/Parts/F_MED_SpeedBonny_FaceAcc_CtrlRig
graphs [
 ContainedGraph {nodes [
   n1 = RigVMFunctionEntryNode()
   RigUnit_GetTransform = &RigUnit_GetTransform(Space=LocalSpace, bInitial=false, Item=@n1.Joint)
   n2 = &RigVMFunction_MathQuaternionToEuler(RotationOrder=ZYX, Value=@RigUnit_GetTransform.Transform.Rotation)
   > RigUnit_SetTransform = &RigUnit_SetTransform(
    Space=LocalSpace
    Value=(Rotation=(X=0.0132,Y=0.4099,Z=-0.0222,W=0.9118),Translation=(X=23.6045,Y=-6.37,Z=140.1115),Scale3D=(X=1,Y=1,Z=1))
    bInitial=false
    Item=@n1.Joint
    Weight=@n1.Weight
    Value.Translation.Z=&RigVMFunction_MathFloatAdd(B=@n1.Translate Z, A=@RigUnit_GetTransform.Transform.Translation.Z) {template="Add::Execute(in A,in B,out Result)"}.Result
    Value.Scale3D.Z=&RigVMFunction_MathFloatAdd(B=@n1.Scale Z, A=@RigUnit_GetTransform.Transform.Scale3D.Z) {template="Add::Execute(in A,in B,out Result)"}.Result
    Value.Translation.Y=&RigVMFunction_MathFloatAdd(B=@n1.Translate Y, A=@RigUnit_GetTransform.Transform.Translation.Y) {template="Add::Execute(in A,in B,out Result)"}.Result
    Value.Scale3D.Y=&RigVMFunction_MathFloatAdd(B=@n1.Scale Y, A=@RigUnit_GetTransform.Transform.Scale3D.Y) {template="Add::Execute(in A,in B,out Result)"}.Result
    Value.Translation.X=&RigVMFunction_MathFloatAdd(B=@n1.Translate X, A=@RigUnit_GetTransform.Transform.Translation.X) {template="Add::Execute(in A,in B,out Result)"}.Result
    Value.Scale3D.X=&RigVMFunction_MathFloatAdd(B=@n1.Scale X, A=@RigUnit_GetTransform.Transform.Scale3D.X) {template="Add::Execute(in A,in B,out Result)"}.Result
    Value.Rotation=&RigVMFunction_MathQuaternionFromEuler(
     Euler=(X=0,Y=0,Z=0)
     Euler.X=&RigVMFunction_MathFloatAdd(B=@n1.Rotate X, A=@n2.Result.X) {template="Add::Execute(in A,in B,out Result)"}.Result
     Euler.Y=&RigVMFunction_MathFloatAdd(B=@n1.Rotate Y, A=@n2.Result.Y) {template="Add::Execute(in A,in B,out Result)"}.Result
     Euler.Z=&RigVMFunction_MathFloatAdd(B=@n1.Rotate Z, A=@n2.Result.Z) {template="Add::Execute(in A,in B,out Result)"}.Result
    ).Result
   )
   > RigVMFunctionReturnNode()
  ]}
 RigVMFunctionLibrary {nodes [Set Transforms for Bone = RigVMCollapseNode(
    ExecuteContext=()
    Joint=()
    Rotate X=0
    Rotate Y=0
    Rotate Z=0
    Scale X=0
    Scale Y=0
    Scale Z=0
    Translate X=0
    Translate Y=0
    Translate Z=0
    Weight=1
   )]}
 RigVMModel {nodes [
   RigUnit_BeginExecution = &RigUnit_BeginExecution()
   n1 = &RigUnit_SphericalPoseReader(
    ActiveRegionScaleFactors=(PositiveWidth=1,NegativeWidth=1,PositiveHeight=1,NegativeHeight=1)
    ActiveRegionSize=0.1
    Debug=(bDrawDebug=false,bDraw2D=false,bDrawLocalAxes=false,DebugScale=25,DebugSegments=12,DebugThickness=0.15)
    DriverAxis=(X=0,Y=1,Z=0)
    DriverItem=(Type=Bone,Name="head")
    FalloffRegionScaleFactors=(PositiveWidth=1,NegativeWidth=3,PositiveHeight=3,NegativeHeight=3)
    FalloffSize=0.2
    FlipWidthScaling=false
    OptionalParentItem=(Type=Bone,Name="spine_05")
    RotationOffset=(X=115,Y=0,Z=-80)
    ExecutePin=@RigUnit_BeginExecution.ExecutePin
   )
   > &RigUnit_ModifyTransforms(
    ItemToModify="((Item=(Type=Bone,Name=\"dyn_strand_bck_MSTR\"),Transform=(Rotation=(X=0,Y=-0.3827,Z=0,W=0.9239),Translation=(X=0,Y=0,Z=0),Scale3D=(X=1,Y=1,Z=1))),(Item=(Type=Bone,Name=\"dyn_strand_side2_MSTR\"),Transform=(Rotation=(X=0,Y=-0.5,Z=0,W=0.866),Translation=(X=0,Y= ...(truncated)"
    WeightMinimum=0
    Weight=@n1.OutputParam
   )
  ]}
]
hierarchy [
 root {loc=<0 0 0>, rot=<0 0 0 1>}
 pelvis {parent=root, loc=<0 -0.5301 98.0571>, rot=<0.0127 0.707 -0.0127 -0.707>}
 spine_01 {parent=pelvis, loc=<6.1208 -0.9328 0>, rot=<0 0 0.034 -0.9994>}
 spine_02 {parent=spine_01, loc=<6.5996 -0.2095 0>, rot=<0 0 0.0238 -0.9997>}
 spine_03 {parent=spine_02, loc=<6.4914 -0.1237 0>, rot=<0 0 0.0149 -0.9999>}
 spine_04 {parent=spine_03, loc=<6.9598 0.0794 0>, rot=<0 0 -0.0132 -0.9999>}
 spine_05 {parent=spine_04, loc=<6.9987 0 0>, rot=<0 0 -0.0015 -1>}
 neck_01 {parent=spine_05, loc=<11.3138 0.1529 0>, rot=<0 0 -0.2003 -0.9797>}
 neck_02 {parent=neck_01, loc=<4.5627 0 0>, rot=<0 0 0.0504 -0.9987>}
 head {parent=neck_02, loc=<5.3914 0 0>, rot=<0 0 0.115 -0.9934>}
 dyn_strand_side2_MSTR {parent=head, loc=<12.9544 -5.5634 5.3255>, rot=<0.0726 -0.3211 0.944 -0.0231>}
 dyn_strand_side2_01 {parent=dyn_strand_side2_MSTR, loc=<0 0 0>, rot=<0 0 0 1>}
 dyn_strand_side2_02 {parent=dyn_strand_side2_01, loc=<14.2622 -0.0014 -0.4197>, rot=<-0.0836 -0.0105 0.001 -0.9964>}
 dyn_strand_bck_MSTR {parent=head, loc=<9.8953 -9.1473 2.3927>, rot=<0.0207 -0.5526 0.8332 -0.0067>}
 dyn_strand_bck_01 {parent=dyn_strand_bck_MSTR, loc=<0 0 0>, rot=<0 0 0 1>}
 dyn_strand_bck_02 {parent=dyn_strand_bck_01, loc=<9.0142 -0.1542 -0.0604>, rot=<-0.0045 0.0003 -0.0068 -1>}
 dyn_strand_bck_03 {parent=dyn_strand_bck_02, loc=<5.3853 0.0555 -0.3956>, rot=<0.0096 0.1856 -0.0093 -0.9825>}
 dyn_strand_side1_MSTR {parent=head, loc=<12.8386 -0.5221 5.5853>, rot=<0.0696 -0.0007 0.9974 -0.018>}
 dyn_strand_side1_01 {parent=dyn_strand_side1_MSTR, loc=<0 0 0>, rot=<0 0 0 1>}
 dyn_strand_side1_02 {parent=dyn_strand_side1_01, loc=<14.0168 0.0256 0.5529>, rot=<-0.0135 -0.0204 0.0008 -0.9997>}
 dyn_strand_fr_MSTR {parent=head, loc=<9.7444 5.6577 6.3534>, rot=<0.0438 0.2593 0.9647 -0.016>}
 dyn_strand_fr_01 {parent=dyn_strand_fr_MSTR, loc=<0 0 0>, rot=<0 0 0 1>}
 dyn_strand_fr_02 {parent=dyn_strand_fr_01, loc=<7.003 -0.086 0.4548>, rot=<-0.019 -0.1053 -0.0178 -0.9941>}
 dyn_strand_fr_03 {parent=dyn_strand_fr_02, loc=<6.0041 0.771 0>, rot=<0.1252 0.064 0.046 -0.989>}
 clavicle_r {parent=spine_05, loc=<5.4831 -0.1053 3.2946>, rot=<0.7289 0.0465 -0.6829 0.009>}
 upperarm_r {parent=clavicle_r, loc=<-9.8658 1.2136 0.4105>, rot=<0.0164 -0.3245 -0.0269 -0.9454>}
 clavicle_l {parent=spine_05, loc=<5.4831 -0.1053 -3.2946>, rot=<0.0465 -0.7289 -0.009 -0.6829>}
]
variables [get_head_look_up_alpha {type=double, default=0}]
