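# Physics asset for the SpeedBonny female medium body's face accessory, simulating its dangling strands.
# Four capsule chains (front, back, two sides) hang from MSTR root bones with tight twist and wider, uneven swing limits.
# Extra constraints with free rotation cross-link neighboring chains at their lower segments, tying the strands together.
# Head, upper spine, both clavicles and only the right upper arm supply simple body shapes around the strands.
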PhysicsAsset /BRCosmetics/Characters/Player/Female/Medium/Bodies/F_MED_SpeedBonny/Meshes/Parts/F_MED_SpeedBonny_FaceAcc_Physics
bodies {
 clavicle_l [{shapeType=Sphere, radius=8, shapeName=clavicle_l_sphere, center=<6.0182 0 0>}]
 clavicle_r [{shapeType=Capsule, length=9, radius=9.5, shapeName=clavicle_r_capsule, center=<-5.7294 0 7.3028>, rotation=<rot 0 179.9995 -179.9994>}]
 dyn_strand_bck_01 [{
   shapeType=Capsule
   length=2.4331
   radius=2.0669
   shapeName=dyn_strand_bck_01_capsule
   center=<4.2708 0 0>
   rotation=<rot 0 -90.0005 -89.9987>
  }]
 dyn_strand_bck_02 [{
   shapeType=Capsule
   length=5.2525
   radius=2.7253
   shapeName=dyn_strand_bck_02_capsule
   center=<3.5608 0 0.2456>
   rotation=<rot -0.0009 -0.0004 -90>
  }]
 dyn_strand_bck_03 [{
   shapeType=Capsule
   length=1.9033
   radius=2
   shapeName=dyn_strand_bck_03_capsule
   center=<6.4516 0.1228 0.2934>
   rotation=<rot -20.0001 90 -90.001>
  }]
 dyn_strand_bck_MSTR [{shapeType=Capsule, length=1, radius=1, shapeName=dyn_strand_bck_MSTR_capsule, rotation=<rot -0.0001 0 -90>}]
 dyn_strand_fr_01 [{
   shapeType=Capsule
   length=2.4331
   radius=1.5669
   shapeName=dyn_strand_fr_01_capsule
   center=<3.7451 -0.0001 0>
   rotation=<rot -0.0001 89.9988 -89.9993>
  }]
 dyn_strand_fr_02 [{
   shapeType=Capsule
   length=6.0885
   radius=1.9115
   shapeName=dyn_strand_fr_02_capsule
   center=<3.3986 0 0.8245>
   rotation=<rot 0.0001 -0.0001 -90>
  }]
 dyn_strand_fr_03 [{shapeType=Capsule, length=5.3495, radius=2.6505, shapeName=dyn_strand_fr_03_capsule, center=<4.6889 0 0.9117>, rotation=<rot 0 -0.0001 -90>}]
 dyn_strand_fr_MSTR [{shapeType=Capsule, length=1, radius=1, shapeName=dyn_strand_fr_MSTR_capsule, rotation=<rot 0 -0.0001 -90>}]
 dyn_strand_side1_01 [{shapeType=Capsule, length=4.5, radius=2, shapeName=dyn_strand_side1_01_capsule, center=<5.6352 0 0>, rotation=<rot -90 -179.6835 179.6832>}]
 dyn_strand_side1_02 [{
   shapeType=Capsule
   length=1.3914
   radius=2.6607
   shapeName=dyn_strand_side1_02_capsule
   center=<4.4717 0 1.0678>
   rotation=<rot 0.0001 -90.0008 -90>
  }]
 dyn_strand_side1_MSTR [{shapeType=Capsule, length=1, radius=1, shapeName=dyn_strand_side1_MSTR_capsule, rotation=<rot 0 -0.0001 -90>}]
 dyn_strand_side2_01 [{
   shapeType=Capsule
   length=4.6573
   radius=1.8427
   shapeName=dyn_strand_side2_01_capsule
   center=<5.4703 0 0>
   rotation=<rot -90 -179.9852 179.9849>
  }]
 dyn_strand_side2_02 [{
   shapeType=Capsule
   length=1.8595
   radius=2.1926
   shapeName=dyn_strand_side2_02_capsule
   center=<5.2347 0.1543 -0.5478>
   rotation=<rot 90 90 -89.9989>
  }]
 dyn_strand_side2_MSTR [{shapeType=Capsule, length=1, radius=1, shapeName=dyn_strand_side2_MSTR_capsule, rotation=<rot 0 -0.0001 -90>}]
 head [{shapeType=Capsule, length=8.7601, radius=8.3194, shapeName=head_capsule, center=<-1.6565 0.7604 -1.2429>, rotation=<rot 90 0 0.0001>}]
 spine_05 [{shapeType=Capsule, length=11, radius=9.7394, shapeName=spine_05_capsule, center=<-1.7252 -0.5524 0>, rotation=<rot 90 -0.0044 -0.0035>}]
 upperarm_r [{shapeType=Capsule, length=16.1825, radius=6.788, shapeName=upperarm_r_capsule, center=<-13.6556 0 0>, rotation=<rot -90 0.0099 -0.0099>}]
}
constraints [
 {bone1Name=dyn_strand_bck_01, bone2Name=dyn_strand_bck_MSTR, swing1LimitDegrees=25, swing2LimitDegrees=60, twistLimitDegrees=2}
 {bone1Name=dyn_strand_bck_02, bone2Name=dyn_strand_bck_01, swing1LimitDegrees=15, swing2LimitDegrees=45, twistLimitDegrees=1}
 {bone1Name=dyn_strand_bck_03, bone2Name=dyn_strand_bck_02, swing1LimitDegrees=10, swing2LimitDegrees=65, twistLimitDegrees=1}
 {bone1Name=dyn_strand_fr_01, bone2Name=dyn_strand_fr_MSTR, swing1LimitDegrees=20, swing2LimitDegrees=45, twistLimitDegrees=2}
 {bone1Name=dyn_strand_fr_02, bone2Name=dyn_strand_fr_01, swing1LimitDegrees=10, swing2LimitDegrees=60, twistLimitDegrees=1}
 {bone1Name=dyn_strand_fr_03, bone2Name=dyn_strand_fr_02, swing1LimitDegrees=10, swing2LimitDegrees=60, twistLimitDegrees=1}
 {bone1Name=dyn_strand_side1_01, bone2Name=dyn_strand_side1_MSTR, swing1LimitDegrees=10, swing2LimitDegrees=45, twistLimitDegrees=2}
 {bone1Name=dyn_strand_side1_02, bone2Name=dyn_strand_side1_01, swing1LimitDegrees=10, swing2LimitDegrees=75, twistLimitDegrees=1}
 {bone1Name=dyn_strand_side2_01, bone2Name=dyn_strand_side2_MSTR, swing1LimitDegrees=10, swing2LimitDegrees=45, twistLimitDegrees=2}
 {bone1Name=dyn_strand_side2_02, bone2Name=dyn_strand_side2_01, swing1LimitDegrees=10, swing2LimitDegrees=75, twistLimitDegrees=1}
 {
  bone1Name=dyn_strand_bck_02
  bone2Name=dyn_strand_side2_02
  swing1LimitDegrees=45
  swing1Motion=Free
  swing2LimitDegrees=45
  swing2Motion=Free
  twistLimitDegrees=45
  twistMotion=Free
 }
 {
  bone1Name=dyn_strand_side1_02
  bone2Name=dyn_strand_side2_02
  swing1LimitDegrees=45
  swing1Motion=Free
  swing2LimitDegrees=45
  swing2Motion=Free
  twistLimitDegrees=45
  twistMotion=Free
 }
 {
  bone1Name=dyn_strand_side1_02
  bone2Name=dyn_strand_fr_03
  swing1LimitDegrees=45
  swing1Motion=Free
  swing2LimitDegrees=45
  swing2Motion=Free
  twistLimitDegrees=45
  twistMotion=Free
 }
 {
  bone1Name=dyn_strand_side2_02
  bone2Name=dyn_strand_bck_03
  swing1LimitDegrees=45
  swing1Motion=Free
  swing2LimitDegrees=45
  swing2Motion=Free
  twistLimitDegrees=45
  twistMotion=Free
 }
]
