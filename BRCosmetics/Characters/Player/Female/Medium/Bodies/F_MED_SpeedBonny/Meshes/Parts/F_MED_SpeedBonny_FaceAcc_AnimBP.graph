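# Post-process style AnimBlueprint for the SpeedBonny female cosmetic's FaceAcc part, animating dynamic hair strands.
# It copies the parent mesh pose, runs a Control Rig and pre-offsets strand roots before RigidBody simulation in neck_02 space.
# Fixed and pose-driven bone corrections follow, using alphas from head and right-arm angles on the base mesh and crouch state.
# Skydive and parachute states retune simulation alpha and inertia, while a deprecated helper overrides gravity.
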AnimBlueprint /BRCosmetics/Characters/Player/Female/Medium/Bodies/F_MED_SpeedBonny/Meshes/Parts/F_MED_SpeedBonny_FaceAcc_AnimBP
target_skeleton=&/F_MED_SpeedBonny_FaceAcc_Skeleton
graphs [
 AnimGraph {nodes [
   n1 = RigidBody!(
    ComponentPose=Local To Component(LocalPose=Use cached pose 'PreRBANPositioning'())
    Alpha=Get RBANAlpha() -> (RBANAlpha:double)
    OverrideWorldGravity:Vector=Get Override World Gravity() -> (Override World Gravity:Vector)
    ExternalForce
    ComponentLinearAccScale:Vector=Get Component Linear Acc Scale() -> (Component Linear Acc Scale:Vector)
    ComponentLinearVelScale:Vector=Get Component Linear Vel Scale() -> (Component Linear Vel Scale:Vector)
   ) -> (Pose) {
    bClampLinearTranslationLimitToRefPose=true
    bOverrideWorldGravity=true
    bTransferBoneVelocities=true
    overridePhysicsAsset=&/F_MED_SpeedBonny_FaceAcc_Physics
    simulationSpace=BaseBoneSpace
    alphaBoolBlend {blendOption=Linear}
    baseBoneRef {boneName=neck_02}
    componentAppliedLinearAccClamp=<1000 1000 1000>
    componentLinearAccScale=<0.4 0.4 0.3>
    componentLinearVelScale=<0.4 0.4 0.3>
    overrideWorldGravity {z=-980}
   }
   n2 = "Transform (Modify) Bone - Bone: dyn_strand_bck_03"!(ComponentPose=@n1, bAlphaBoolEnabled:bool=True, Translation="0,0,0", Rotation="0,0,0", Scale="0.5,0.8,0.8") -> (Pose) {
    alphaInputType=Bool
    rotationSpace=BCS_BoneSpace
    scaleMode=BMM_Replace
    scaleSpace=BCS_BoneSpace
    translationSpace=BCS_ParentBoneSpace
    alphaBoolBlend {blendInTime=0.5, blendOption=CubicInOut, blendOutTime=0.5}
    boneToModify {boneName=dyn_strand_bck_03}
    scale=<0.5 0.8 0.8>
   }
   n3 = "Transform (Modify) Bone - Bone: dyn_strand_side1_01"!(ComponentPose=@n2, bAlphaBoolEnabled:bool=True, Translation="0,0,0", Rotation="0,0,0", Scale="0.8,1,1") -> (Pose) {
    alphaInputType=Bool
    rotationSpace=BCS_BoneSpace
    scaleMode=BMM_Replace
    scaleSpace=BCS_BoneSpace
    translationSpace=BCS_ParentBoneSpace
    alphaBoolBlend {blendInTime=0.5, blendOption=CubicInOut, blendOutTime=0.5}
    boneToModify {boneName=dyn_strand_side1_01}
    scale {x=0.8}
   }
   n4 = "Transform (Modify) Bone - Bone: dyn_strand_side1_02"!(ComponentPose=@n3, bAlphaBoolEnabled:bool=True, Translation="0,0,0", Rotation="0,0,0", Scale="0.6,1,1") -> (Pose) {
    alphaInputType=Bool
    rotationSpace=BCS_BoneSpace
    scaleMode=BMM_Replace
    scaleSpace=BCS_BoneSpace
    translationSpace=BCS_ParentBoneSpace
    alphaBoolBlend {blendInTime=0.5, blendOption=CubicInOut, blendOutTime=0.5}
    boneToModify {boneName=dyn_strand_side1_02}
    scale {x=0.6}
   }
   n5 = "Transform (Modify) Bone - Bone: dyn_strand_side2_01"!(ComponentPose=@n4, bAlphaBoolEnabled:bool=True, Translation="0,0,0", Rotation="0,0,0", Scale="0.9,1,1") -> (Pose) {
    alphaInputType=Bool
    rotationSpace=BCS_BoneSpace
    scaleMode=BMM_Replace
    scaleSpace=BCS_BoneSpace
    translationSpace=BCS_ParentBoneSpace
    alphaBoolBlend {blendInTime=0.5, blendOption=CubicInOut, blendOutTime=0.5}
    boneToModify {boneName=dyn_strand_side2_01}
    scale {x=0.9}
   }
   n6 = "Transform (Modify) Bone - Bone: dyn_strand_side2_02"!(ComponentPose=@n5, bAlphaBoolEnabled:bool=True, Translation="0,0,0", Rotation="0,0,0", Scale="0.8,1,1") -> (Pose) {
    alphaInputType=Bool
    rotationSpace=BCS_BoneSpace
    scaleMode=BMM_Replace
    scaleSpace=BCS_BoneSpace
    translationSpace=BCS_ParentBoneSpace
    alphaBoolBlend {blendInTime=0.5, blendOption=CubicInOut, blendOutTime=0.5}
    boneToModify {boneName=dyn_strand_side2_02}
    scale {x=0.8}
   }
   n7 = "Transform (Modify) Bone - Bone: dyn_strand_fr_02"!(ComponentPose=@n6, bAlphaBoolEnabled:bool=True, Translation="0,0,0", Rotation="0,0,0", Scale="0.9,1,1") -> (Pose) {
    alphaInputType=Bool
    rotationSpace=BCS_BoneSpace
    scaleMode=BMM_Replace
    scaleSpace=BCS_BoneSpace
    translationSpace=BCS_ParentBoneSpace
    alphaBoolBlend {blendInTime=0.5, blendOption=CubicInOut, blendOutTime=0.5}
    boneToModify {boneName=dyn_strand_fr_02}
    scale {x=0.9}
   }
   n8 = "Transform (Modify) Bone - Bone: dyn_strand_fr_03"!(ComponentPose=@n7, bAlphaBoolEnabled:bool=True, Translation="0,0,0", Rotation="0,0,0", Scale="0.8,1,1") -> (Pose) {
    alphaInputType=Bool
    rotationSpace=BCS_BoneSpace
    scaleMode=BMM_Replace
    scaleSpace=BCS_BoneSpace
    translationSpace=BCS_ParentBoneSpace
    alphaBoolBlend {blendInTime=0.5, blendOption=CubicInOut, blendOutTime=0.5}
    boneToModify {boneName=dyn_strand_fr_03}
    scale {x=0.8}
   }
   n9 = "Transform (Modify) Bone - Bone: dyn_strand_bck_02"(ComponentPose=@n8, Alpha=Get HeadUpCrouchedAlpha() -> (HeadUpCrouchedAlpha:double), Translation="0,0,1", Rotation="-25,0,0", Scale="1,1,1") {
    rotationMode=BMM_Additive
    rotationSpace=BCS_BoneSpace
    scaleSpace=BCS_BoneSpace
    translationMode=BMM_Additive
    translationSpace=BCS_BoneSpace
    alphaBoolBlend {blendInTime=0.5, blendOption=CubicInOut, blendOutTime=0.5}
    alphaScaleBiasClamp {bInterpResult=true}
    boneToModify {boneName=dyn_strand_bck_02}
    rotation {pitch=-25}
    translation {z=1}
   }
   n10 = Get RightUpperArmUpAlpha() -> (RightUpperArmUpAlpha:float)
   n11 = "Transform (Modify) Bone - Bone: dyn_strand_fr_02"!(ComponentPose=@n9, Alpha:float=@n10, Translation="0,0,0", Rotation="20,-5,0", Scale="1,1,1") -> (Pose) {
    rotationMode=BMM_Additive
    rotationSpace=BCS_BoneSpace
    alphaBoolBlend {blendOption=Linear}
    boneToModify {boneName=dyn_strand_fr_02}
    rotation {pitch=20, yaw=-5}
   }
   n12 = "Transform (Modify) Bone - Bone: dyn_strand_fr_03"!(ComponentPose=@n11, Alpha:float=@n10, Translation="0,0,0", Rotation="-5,-5,0", Scale="1,1,1") -> (Pose) {
    rotationMode=BMM_Additive
    rotationSpace=BCS_BoneSpace
    alphaBoolBlend {blendOption=Linear}
    boneToModify {boneName=dyn_strand_fr_03}
    rotation {pitch=-5, yaw=-5}
   }
   n13 = Get BackHairDownSpineAlpha() -> (BackHairDownSpineAlpha:float)
   n14 = "Transform (Modify) Bone - Bone: dyn_strand_bck_01"(ComponentPose=@n12, Alpha=@n13, Translation="0,0,0", Rotation="-5,20,0", Scale="1,1,1") {
    rotationMode=BMM_Additive
    rotationSpace=BCS_BoneSpace
    scaleSpace=BCS_BoneSpace
    translationMode=BMM_Additive
    translationSpace=BCS_BoneSpace
    alphaBoolBlend {blendInTime=0.5, blendOption=CubicInOut, blendOutTime=0.5}
    alphaScaleBiasClamp {bInterpResult=true}
    boneToModify {boneName=dyn_strand_bck_01}
    rotation {pitch=-5, yaw=20}
   }
   n15 = "Transform (Modify) Bone - Bone: dyn_strand_bck_02"(ComponentPose=@n14, Alpha=@n13, Translation="0,-2,0", Rotation="-5,0,0", Scale="1,1,1") {
    rotationMode=BMM_Additive
    rotationSpace=BCS_BoneSpace
    scaleSpace=BCS_BoneSpace
    translationMode=BMM_Additive
    translationSpace=BCS_BoneSpace
    alphaBoolBlend {blendInTime=0.5, blendOption=CubicInOut, blendOutTime=0.5}
    alphaScaleBiasClamp {bInterpResult=true}
    boneToModify {boneName=dyn_strand_bck_02}
    rotation {pitch=-5}
    translation {y=-2}
   }
   n16 = "Transform (Modify) Bone - Bone: dyn_strand_bck_01"(ComponentPose=@n15, Alpha=Get BackHairOnArmAlpha() -> (BackHairOnArmAlpha:float), Translation="0,0,1.5", Rotation="0,0,0", Scale="1,1,1") {
    translationMode=BMM_Additive
    translationSpace=BCS_BoneSpace
    alphaBoolBlend {blendOption=Linear}
    boneToModify {boneName=dyn_strand_bck_01}
    translation {z=1.5}
   }
   n17 = "Transform (Modify) Bone - Bone: dyn_strand_bck_02"!(ComponentPose=@n16, bAlphaBoolEnabled:bool=True, Translation="0,0,0", Rotation="9.999,-20.001,0", Scale="1,1,1") -> (Pose) {
    alphaInputType=Bool
    rotationMode=BMM_Additive
    rotationSpace=BCS_BoneSpace
    translationMode=BMM_Additive
    translationSpace=BCS_ParentBoneSpace
    alphaBoolBlend {blendInTime=0.5, blendOption=CubicInOut, blendOutTime=0.5}
    boneToModify {boneName=dyn_strand_bck_02}
    rotation {pitch=9.999, yaw=-20.001}
   }
   n18 = "Transform (Modify) Bone - Bone: dyn_strand_bck_03"(ComponentPose=@n17, bAlphaBoolEnabled=True, Translation="0,0,0", Rotation="-30,0,0", Scale="0.7,1,1") {
    alphaInputType=Bool
    rotationMode=BMM_Additive
    rotationSpace=BCS_BoneSpace
    scaleMode=BMM_Replace
    translationMode=BMM_Additive
    translationSpace=BCS_ParentBoneSpace
    alphaBoolBlend {blendInTime=0.5, blendOption=CubicInOut, blendOutTime=0.5}
    boneToModify {boneName=dyn_strand_bck_03}
    rotation {pitch=-30}
    scale {x=0.7}
   }
   n19 = "Transform (Modify) Bone - Bone: dyn_strand_side2_02"(ComponentPose=@n18, bAlphaBoolEnabled=True, Translation="0,0,0", Rotation="-10,-10,0", Scale="1,1,1") {
    alphaInputType=Bool
    rotationMode=BMM_Additive
    rotationSpace=BCS_BoneSpace
    translationMode=BMM_Additive
    translationSpace=BCS_ParentBoneSpace
    alphaBoolBlend {blendInTime=0.5, blendOption=CubicInOut, blendOutTime=0.5}
    boneToModify {boneName=dyn_strand_side2_02}
    rotation {pitch=-10, yaw=-10}
   }
   n20 = "Transform (Modify) Bone - Bone: dyn_strand_side1_02"(ComponentPose=@n19, bAlphaBoolEnabled=True, Translation="0,0,0", Rotation="-10,-10,0", Scale="1,1,1") {
    alphaInputType=Bool
    rotationMode=BMM_Additive
    rotationSpace=BCS_BoneSpace
    translationMode=BMM_Additive
    translationSpace=BCS_ParentBoneSpace
    alphaBoolBlend {blendInTime=0.5, blendOption=CubicInOut, blendOutTime=0.5}
    boneToModify {boneName=dyn_strand_side1_02}
    rotation {pitch=-10, yaw=-10}
   }
   n21 = "Transform (Modify) Bone - Bone: dyn_strand_fr_02"!(ComponentPose=@n20, bAlphaBoolEnabled:bool=True, Translation="0,0,0", Rotation="-10,0,0", Scale="1,1,1") -> (Pose) {
    alphaInputType=Bool
    rotationMode=BMM_Additive
    rotationSpace=BCS_BoneSpace
    translationMode=BMM_Additive
    translationSpace=BCS_ParentBoneSpace
    alphaBoolBlend {blendInTime=0.5, blendOption=CubicInOut, blendOutTime=0.5}
    boneToModify {boneName=dyn_strand_fr_02}
    rotation {pitch=-10}
   }
   Output Pose(Result=Component To Local(ComponentPose=@n21)) {layerGroup=None}
   Save cached pose 'InputPose'(Pose=Copy Pose From Mesh() {bUseAttachedParent=true}) {cacheName=InputPose}
   n22 = Control Rig(Source=Use cached pose 'InputPose'(), Alpha=Get RBANAlpha()) {
    alphaBoolBlend {blendOption=Linear}
    controlRigAssetReference {blueprintRigClass=&/F_MED_SpeedBonny_FaceAcc_CtrlRig.F_MED_SpeedBonny_FaceAcc_CtrlRig_C}
   }
   n23 = "Transform (Modify) Bone - Bone: dyn_strand_bck_MSTR"!(
    ComponentPose=Local To Component(LocalPose=@n22)
    bAlphaBoolEnabled:bool=True
    Translation="0,1,1"
    Rotation="-14.106,-44.108,-0.110"
    Scale="1,1,1"
   ) -> (Pose) {
    alphaInputType=Bool
    rotationMode=BMM_Additive
    rotationSpace=BCS_BoneSpace
    translationMode=BMM_Additive
    translationSpace=BCS_ParentBoneSpace
    alphaBoolBlend {blendInTime=0.5, blendOption=CubicInOut, blendOutTime=0.5}
    boneToModify {boneName=dyn_strand_bck_MSTR}
    rotation=<rot -14.106 -44.108 -0.11>
    translation {y=1, z=1}
   }
   n24 = "Transform (Modify) Bone - Bone: dyn_strand_side2_MSTR"!(ComponentPose=@n23, bAlphaBoolEnabled:bool=True, Translation="0,0,0", Rotation="-0.000,-40.000,0.000", Scale="1,1,1") -> (Pose) {
    alphaInputType=Bool
    rotationMode=BMM_Additive
    rotationSpace=BCS_BoneSpace
    translationMode=BMM_Additive
    translationSpace=BCS_ParentBoneSpace
    alphaBoolBlend {blendInTime=0.5, blendOption=CubicInOut, blendOutTime=0.5}
    boneToModify {boneName=dyn_strand_side2_MSTR}
    rotation=<rot 0 -40 0>
   }
   n25 = "Transform (Modify) Bone - Bone: dyn_strand_side1_MSTR"!(ComponentPose=@n24, bAlphaBoolEnabled:bool=True, Translation="0,0,0", Rotation="0.000,-40.000,0.000", Scale="1,1,1") -> (Pose) {
    alphaInputType=Bool
    rotationMode=BMM_Additive
    rotationSpace=BCS_BoneSpace
    translationMode=BMM_Additive
    translationSpace=BCS_ParentBoneSpace
    alphaBoolBlend {blendInTime=0.5, blendOption=CubicInOut, blendOutTime=0.5}
    boneToModify {boneName=dyn_strand_side1_MSTR}
    rotation=<rot 0 -40 0>
   }
   n26 = "Transform (Modify) Bone - Bone: dyn_strand_fr_MSTR"!(ComponentPose=@n25, bAlphaBoolEnabled:bool=True, Translation="0,0,0", Rotation="0,-30,0", Scale="1,1,1") -> (Pose) {
    alphaInputType=Bool
    rotationMode=BMM_Additive
    rotationSpace=BCS_BoneSpace
    translationMode=BMM_Additive
    translationSpace=BCS_ParentBoneSpace
    alphaBoolBlend {blendInTime=0.5, blendOption=CubicInOut, blendOutTime=0.5}
    boneToModify {boneName=dyn_strand_fr_MSTR}
    rotation {yaw=-30}
   }
   Save cached pose 'PreRBANPositioning'(Pose=Component To Local(ComponentPose=@n26)) {cacheName=PreRBANPositioning}
  ]}
 EventGraph {nodes [
   Event BlueprintUpdateAnimation()
   > n1 = Sequence()
   > K2Node_Composite_2 = Variables() {is=title K2Node_Composite}
   K2Node_Composite_0 = Gravity Override(execute=@n1.then_1) {is=title K2Node_Composite}
   n2 = Event OnCharacterCustomizationLoaded()
   > Set Custom Base Mesh(Custom Base Mesh:Skeletal Mesh Component Object Reference=Get Array Item with EFort Custom Part Type(Array=@n2.CustomMeshParts)) -> (Output_Get:Skeletal Mesh Component Object Reference)
   Is Valid(InputObject=Get Custom Base Mesh() -> (Custom Base Mesh:Skeletal Mesh Component Object Reference))
   n3 = BreakTransform(InTransform=GetSocketTransform(self=Get Custom Base Mesh(), InSocketName=upperarm_r, TransformSpace=RTS_World))
   > Set Translation(Translation:Vector=@n3) -> (Output_Get:Vector)
   > Set Rotation_0(Rotation_0:Rotator=@n3.Rotation) -> (Output_Get:Rotator)
   n4 = BreakTransform(InTransform=GetSocketTransform(self=Get Custom Base Mesh(), InSocketName=clavicle_r, TransformSpace=RTS_World))
   > Set Translation_0(Translation_0:Vector=@n4) -> (Output_Get:Vector)
   > Set Rotation_1(Rotation_1:Rotator=@n4.Rotation) -> (Output_Get:Rotator)
   n5 = BreakTransform(InTransform=GetSocketTransform(self=Get Custom Base Mesh(), InSocketName=spine_05, TransformSpace=RTS_World))
   > Set Translation_1(Translation_1:Vector=@n5) -> (Output_Get:Vector)
   > Set Rotation_2(Rotation_2:Rotator=@n5.Rotation) -> (Output_Get:Rotator)
  ]}
 Gravity Override {
  path=EventGraph.K2Node_Composite_0.Gravity Override
  nodes [
   Inputs()
   > n1 = Gravity Override State Function DEPRECATED(CharacterPartAnimInstance=Self-Reference(), GameTimeInSeconds=GetGameTimeInSeconds(), GravityOverrideParams=Get Gravity Override Params())
   > Set Override World Gravity(Override World Gravity:Vector=@n1) -> (Output_Get:Vector)
   Outputs()
  ]
 }
 Variables {
  path=EventGraph.K2Node_Composite_2.Variables
  nodes [
   Inputs()
   > n1 = Sequence!(execute) -> (then_0, then_1, then_2:Exec)
   > n2 = Branch(Condition=Get bIsSkyDiving())
   > Set Component Linear Acc Scale(Component Linear Acc Scale:Vector="0.07,0.07,0.07") -> (Output_Get:Vector)
   > Set Component Linear Vel Scale(Component Linear Vel Scale:Vector="0.1,0.1,0.1") -> (Output_Get:Vector)
   Set Component Linear Vel Scale(execute=@n2.else, Component Linear Vel Scale="0.1,0.1,0.15")
   > Set Component Linear Acc Scale(Component Linear Acc Scale="0.4,0.4,0.4")
   n3 = Get bIsParachuteOpened()
   Set RBANAlpha(execute=@n1.then_1, RBANAlpha:double=SelectFloat(A=SelectFloat(A=0.4, B=0.75, bPickA=@n3), B=1, bPickA=Get bIsSkyDiving())) -> (Output_Get:double)
   n4 = Get Custom Base Mesh()
   Set HeadUpCrouchedAlpha(
    execute=@n1.then_2
    HeadUpCrouchedAlpha:double=SelectFloat(
     A=GetAlphaFromReferenceTransformDelta(SkeletalMeshComponent=@n4, BoneName=head, BaseName=spine_05, DeltaComponentType=EulerY, InputRange=(X=20,Y=50), TwistAxis="1,0,0")
     B=0
     bPickA=Get bIsCrouching()
    )
   ) -> (Output_Get:double)
   n5 = GetAlphaFromReferenceTransformDelta(
    SkeletalMeshComponent=Get Custom Base Mesh()
    BoneName=head
    BaseName=clavicle_r
    DeltaComponentType=EulerY
    InputRange=(X=10,Y=50)
    TwistAxis="1,0,0"
   )
   > Set BackHairDownSpineAlpha(BackHairDownSpineAlpha:float=SelectFloat(A=float * float(A=@n5, B=0.45) {is=title K2Node_PromotableOperator}, B=@n5, bPickA=Get bIsCrouching())) -> (Output_Get:float)
   n6 = Get Custom Base Mesh()
   > Set BackHairOnArmAlpha(BackHairOnArmAlpha:float=SelectFloat(
     A=GetAlphaFromReferenceTransformDelta(SkeletalMeshComponent=@n6, BoneName=head, BaseName=upperarm_r, DeltaComponentType=EulerX, InputRange=(X=-30,Y=-10), TwistAxis="1,0,0")
     B=0
     bPickA=Get bIsCrouching()
    )) -> (Output_Get:float)
   > Set RightUpperArmUpAlpha(RightUpperArmUpAlpha:float=GetAlphaFromReferenceTransformDelta(SkeletalMeshComponent=Get Custom Base Mesh(), BoneName=upperarm_r, BaseName=spine_05, InputRange=(X=120,Y=125), TwistAxis="1,0,0")) -> (Output_Get:float)
   Outputs()
  ]
 }
]
variables [
 BackHairDownSpineAlpha {type=number}
 BackHairOnArmAlpha {type=number}
 Component Linear Acc Scale {type=Vector, default=<0.4 0.4 0.3>}
 Component Linear Vel Scale {type=Vector, default=<0.4 0.4 0.3>}
 Custom Base Mesh {type=SkeletalMeshComponent}
 Gravity Override Params {
  type=GravityOverrideParamsStruct
  default {
   skydive_DivingDown {gravityOverride=<0 0 0>, windAmplitude=<0 0 0>, windFrequency=<0 0 0>}
   skydive_FloatingDown {gravityOverride=<0 0 -700>, windAmplitude=<0 0 0>, windFrequency=<0 0 0>, windOffset=<0 0 0>}
   skydive_GoingUp {gravityOverride=<0 0 0>, windAmplitude=<0 0 0>, windFrequency=<0 0 0>}
   skydive_Parachuting {gravityOverride=<0 -300 -980>, windAmplitude=<200 -200 200>, windFrequency=<1.5 3 2>}
  }
 }
 HeadUpCrouchedAlpha {type=number}
 Override World Gravity {type=Vector, default=<0 0 -980>}
 RBANAlpha {type=number}
 RightUpperArmUpAlpha {type=number}
 Rotation_0 {type=Rotator}
 Rotation_1 {type=Rotator}
 Rotation_2 {type=Rotator}
 Translation {type=Vector}
 Translation_0 {type=Vector}
 Translation_1 {type=Vector}
]
